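# Translucent two-sided wind trail material for the TroutWrist glider's Niagara mesh and ribbon FX.
# A rotated, panning smoke texture drives opacity, shaped by dynamic erode and alpha, particle alpha, UV end fades and camera depth fade, with depth fade skipped on mobile.
# SplineThicken supplies width offset and normals, alongside UV-decoded pivot, axis and random values.
# A vertex-alpha-masked velocity parameter builds an opposing character-movement offset.
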Material /BRCosmetics/Gadgets/Assets/VinderTech_GliderChute/Glider_TroutWrist/FX/Materials/M_WindTrail
graph {
 properties {
  bAutomaticallySetUsageInEditor=false
  bUsedWithNiagaraMeshParticles=true
  bUsedWithNiagaraRibbons=true
  blendMode=BLEND_Translucent
  pixelDepthOffsetMode=PDOM_Legacy
  twoSided=true
  usedShadingModels=Default Lit
 }
 nodes [
  Add(A=Add(A=NamedRerouteUsage(), B=NamedRerouteUsage()), B=NamedRerouteUsage())
  n1 = ParticleColor()
  MaterialFunctionCall(Emissive=@n1.RGB) {materialFunction=&MF_FX_EmissiveAdjustments}
  n2 = Add(A=OneMinus(None=NamedRerouteUsage())) {constB=0.5}
  n3 = MaterialFunctionCall(
   WidthBase=Multiply(A=ScalarParameter() {defaultValue=25, group=WPO, parameterName=WidthBase}, B=@n2)
   WidthTip=Multiply(A=ScalarParameter() {defaultValue=10, group=WPO, parameterName=WidthTip}, B=@n2)
   DeriveNormalZ=StaticBool()
   AngleCorrectedNormals=StaticBool()
   AdditionalNormal=VertexNormalWS()
   FlattenPixelNormal=StaticBool() {value=true}
   UVs For NormalShading=TextureCoordinate()
  ) {materialFunction=&SplineThicken}
  NamedRerouteDeclaration(None=@n3.WorldPositionOffset) {name=SplineThicken_WPO, nodeColor=<rgba 0 1 0.6016 1>}
  NamedRerouteDeclaration(None=@n3.Normal) {name=SplineThicken_Normal, nodeColor=<rgba 1 0.2812 0 1>}
  n4 = TextureCoordinate() {coordinateIndex=2}
  n5 = ComponentMask(None=TextureCoordinate() {coordinateIndex=3}) {r=true}
  n6 = Normalize(VectorInput=ConstantBiasScale(None=AppendVector(A=@n4, B=@n5)) {bias=-0.5, scale=2})
  NamedRerouteDeclaration(None=Normalize(VectorInput=Transform(None=@n6) {transformSourceType=TRANSFORMSOURCE_Local})) {name=Forward Axis, nodeColor=<rgba 1 0 0.1172 1>}
  n7 = TextureCoordinate() {coordinateIndex=2}
  NamedRerouteDeclaration(None=ConstantBiasScale(None=ComponentMask(None=@n7) {g=true}) {bias=-0.5, scale=2}) {name=Random Per Element, nodeColor=<rgba 1 0.6797 0 1>}
  NamedRerouteDeclaration(None=ComponentMask(None=VertexColor().A) {r=true}) {name=Vertex Mask, nodeColor=<rgba 0.3594 0 1 1>}
  n8 = DynamicParameter() {parameterIndex=1, defaultValue=<rgba 1 0.0125 1 0.0125>, paramNames [VelX,VelY,VelZ,VelInfluence]}
  n9 = Power(Base=NamedRerouteUsage(), Exp=ScalarParameter() {defaultValue=2, group=WPO, parameterName=VertexMask Pow})
  n10 = Multiply(B=@n8.RGB) {constA=-1}
  NamedRerouteDeclaration(None=Multiply(A=@n8.VelInfluence, B=Multiply(A=@n9, B=@n10)) {constA=0.0125}) {name=CharacterMovement, nodeColor=<rgba 0 1 0.0156 1>}
  n11 = AppendVector(A=TextureCoordinate(), B=ComponentMask(None=TextureCoordinate() {coordinateIndex=1}) {r=true})
  n12 = Constant3Vector() {constant=<rgba 0 -1 0 0>}
  NamedRerouteDeclaration(None=TransformPosition(None=Add(A=@n11, B=@n12)) {transformType=TRANSFORMPOSSOURCE_World}) {name=PivotPosition, nodeColor=<rgba 0 1 0.2031 1>}
  n13 = TextureCoordinate()
  OneMinus(None=ComponentMask(None=@n13) {r=true})
  n14 = DynamicParameter() {defaultValue=<rgba 1 0 1 1>, paramNames [PanSpeed,Erode,Emissive,MasterAlpha]}
  n15 = AppendVector(A=ScalarParameter() {defaultValue=1, parameterName=UMultiply}, B=ScalarParameter() {defaultValue=1, parameterName=VMultiply})
  n16 = MaterialFunctionCall!(Angle=Constant() {r=0.25}, UV/2DVector=TextureCoordinate()) {materialFunction=/Game/Characters/Player/Male/Medium/Bodies/M_MED_Faux/Materials/MF_2DRotate}
  n17 = Time()
  n18 = ScalarParameter() {defaultValue=0.6, parameterName=U_PannerSpeed}
  n19 = Constant()
  n20 = TextureSampleParameter2D(UVs=Panner(Coordinate=Multiply(A=@n15, B=@n16.Result), Time=Multiply(A=@n14.PanSpeed, B=@n17), Speed=AppendVector(A=@n18, B=@n19)) {speedX=1}) {
   parameterName=TrailTexture
   samplerSource=SSM_Wrap_WorldGroupSettings
   samplerType=SAMPLERTYPE_Grayscale
   texture=&T_Smoke_Trail_Soft
  }
  n21 = Multiply(A=LinearInterpolate(Alpha=Add(A=@n14.Erode, B=@n20.G)) {constA=-3})
  n22 = Multiply(A=Multiply(A=@n1.A, B=Clamp(None=@n21)), B=@n14.MasterAlpha)
  n23 = ComponentMask(None=@n13) {g=true}
  n24 = Power(Base=OneMinus(None=@n23)) {constExponent=1}
  n25 = Multiply(
   A=Multiply(A=Multiply(A=@n22, B=@n24), B=Power(Base=@n23))
   B=MaterialFunctionCall(
    Fade Length=ScalarParameter() {defaultValue=150, parameterName=Fade Length}
    Fade Offset=ScalarParameter() {defaultValue=25, parameterName=Fade Offset}
   ) {materialFunction=&CameraDepthFade}.Result
  )
  n26 = DepthFade(Opacity=@n25, FadeDistance=ScalarParameter() {defaultValue=40, parameterName=DepthFadeDist}) {fadeDistanceDefault=40}
  ShadingPathSwitch(Default=@n26, Deferred=@n26, Mobile=@n25) {desc=Migrated from FeatureLevelSwitch}
 ]
}
