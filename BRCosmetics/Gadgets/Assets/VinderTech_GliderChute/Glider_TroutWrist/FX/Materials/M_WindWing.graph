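# M_WindWing is a translucent, two-sided unlit Niagara mesh-particle material for wind streaks on the TroutWrist glider.
# Particle color, optionally angle-faded, feeds emissive adjustments, while eroded panning smoke, rotatable ramp masks and camera/depth fades build opacity.
# SplineThicken, UV-encoded pivots and axes, and vertex-masked velocity provide world position offset terms.
# Two dynamic parameters expose pan speed, erosion, master alpha and velocity to the particle system.
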Material /BRCosmetics/Gadgets/Assets/VinderTech_GliderChute/Glider_TroutWrist/FX/Materials/M_WindWing
graph {
 properties {
  bAutomaticallySetUsageInEditor=false
  bUsedWithNiagaraMeshParticles=true
  blendMode=BLEND_Translucent
  shadingModel=MSM_Unlit
  twoSided=true
  usedShadingModels=Unlit
 }
 nodes [
  n1 = NamedRerouteUsage()
  n2 = NamedRerouteUsage()
  Add(A=Add(A=@n1, B=@n2), B=NamedRerouteUsage())
  n3 = MaterialFunctionCall("FadeSensitivity (optional)"=ScalarParameter() {defaultValue=1, parameterName=AngleFade}) {materialFunction=&FadeBasedOnViewAngle}
  n4 = Constant() {r=1}
  n5 = ParticleColor()
  MaterialFunctionCall(Emissive=Multiply(A=StaticSwitchParameter(True=@n3.Result, False=@n4) {parameterName=Use Angle Fade}, B=@n5.RGB)) {materialFunction=&MF_FX_EmissiveAdjustments}
  n6 = TextureCoordinate()
  n7 = MaterialFunctionCall!(
   Angle=MaterialFunctionCall!(Degrees=ScalarParameter() {group=Ramp Mask, parameterName="Mask Rotate UVs (Degrees)"}) {materialFunction=/Engine/Functions/Engine_MaterialFunctions01/Math/DegreesToRadians}.Radians
   UV/2DVector=@n6
  ) {materialFunction=/Game/Characters/Player/Male/Medium/Bodies/M_MED_Faux/Materials/MF_2DRotate}
  n8 = Add(A=ComponentMask(None=@n7.Result) {r=true}, B=ScalarParameter() {group=Ramp Mask, parameterName=Mask Offset U})
  n9 = StaticSwitchParameter(True=OneMinus(None=@n8), False=@n8) {group=Ramp Mask, parameterName=Flip Mask U}
  n10 = ScalarParameter() {defaultValue=2, group=Ramp Mask, parameterName=Mask U Power}
  n11 = Add(A=ComponentMask(None=@n7.Result) {g=true}, B=ScalarParameter() {group=Ramp Mask, parameterName=Mask Offset V})
  n12 = StaticSwitchParameter(True=OneMinus(None=@n11), False=@n11) {group=Ramp Mask, parameterName=Flip Mask V}
  n13 = ScalarParameter() {defaultValue=2, group=Ramp Mask, parameterName=Mask V Power}
  n14 = Power(
   Base=Multiply(A=Power(Base=@n9, Exp=@n10) {constExponent=1}, B=Power(Base=@n12, Exp=@n13) {constExponent=1})
   Exp=ScalarParameter() {defaultValue=2, group=Ramp Mask, parameterName=Mask Final Power}
  ) {constExponent=4}
  n15 = MaterialFunctionCall(Float2=@n6) {materialFunction=&BreakOutFloat2Components}
  n16 = Multiply(A=Multiply(A=@n15.R, B=OneMinus(None=@n15.R))) {constB=4}
  Multiply(A=@n14, B=Power(Base=Saturate(None=@n16)) {constExponent=4})
  n17 = NamedRerouteUsage()
  n18 = MaterialFunctionCall(Angle=ScalarParameter() {parameterName=Mask Rotate}, UV/2DVector=TextureCoordinate()) {materialFunction=&MF_2DRotate}
  n19 = Constant2Vector() {g=0.1, r=2}
  n20 = Constant()
  n21 = NamedRerouteUsage()
  n22 = TextureSample(UVs=Panner(
    Coordinate=Multiply(A=@n18.Result, B=@n19)
    Time=Time()
    Speed=MaterialFunctionCall(X=@n20, Y=@n21) {materialFunction=&MakeFloat2}.Result
   ) {speedY=2}) {texture=/Game/Global/Textures/Default/Tileable_Patterns/T_Smoke_Puff_Tile_Up}
  n23 = Multiply(A=LinearInterpolate(Alpha=Add(A=@n17, B=@n22.R)) {constA=-1})
  n24 = Multiply(A=Saturate(None=Clamp(None=@n23)), B=Saturate(None=@n14))
  n25 = Multiply(
   A=Multiply(A=Multiply(A=@n5.A, B=@n24), B=NamedRerouteUsage())
   B=MaterialFunctionCall(Fade Length=ScalarParameter() {defaultValue=150, group=Alpha, parameterName=CamDepthFadeLength}) {materialFunction=&CameraDepthFade}.Result
  )
  Multiply(B=@n25)
  Multiply(A=@n1, B=@n2)
  n26 = Multiply(A=Multiply(A=@n15.G, B=OneMinus(None=@n15.G))) {constB=4}
  Multiply(A=@n24, B=Power(Base=Saturate(None=@n26)) {constExponent=4})
  n27 = Add(A=OneMinus(None=NamedRerouteUsage())) {constB=0.5}
  n28 = MaterialFunctionCall(
   WidthBase=Multiply(A=ScalarParameter() {defaultValue=25, group=WPO, parameterName=WidthBase}, B=@n27)
   WidthTip=Multiply(A=ScalarParameter() {defaultValue=10, group=WPO, parameterName=WidthTip}, B=@n27)
   DeriveNormalZ=StaticBool()
   AngleCorrectedNormals=StaticBool()
   AdditionalNormal=VertexNormalWS()
   FlattenPixelNormal=StaticBool() {value=true}
   UVs For NormalShading=TextureCoordinate()
  ) {materialFunction=&SplineThicken}
  NamedRerouteDeclaration(None=@n28.WorldPositionOffset) {name=SplineThicken_WPO, nodeColor=<rgba 0 1 0.6016 1>}
  NamedRerouteDeclaration(None=@n28.Normal) {name=SplineThicken_Normal, nodeColor=<rgba 1 0.2812 0 1>}
  n29 = TextureCoordinate() {coordinateIndex=2}
  n30 = ComponentMask(None=TextureCoordinate() {coordinateIndex=3}) {r=true}
  n31 = Normalize(VectorInput=ConstantBiasScale(None=AppendVector(A=@n29, B=@n30)) {bias=-0.5, scale=2})
  NamedRerouteDeclaration(None=Normalize(VectorInput=Transform(None=@n31) {transformSourceType=TRANSFORMSOURCE_Local})) {name=Forward Axis, nodeColor=<rgba 1 0 0.1172 1>}
  n32 = TextureCoordinate() {coordinateIndex=2}
  NamedRerouteDeclaration(None=ConstantBiasScale(None=ComponentMask(None=@n32) {g=true}) {bias=-0.5, scale=2}) {name=Random Per Element, nodeColor=<rgba 1 0.6797 0 1>}
  NamedRerouteDeclaration(None=ComponentMask(None=VertexColor().A) {a=true}) {name=Vertex Mask, nodeColor=<rgba 0.3594 0 1 1>}
  n33 = DynamicParameter() {defaultValue=<rgba 1 0.5 1 1>, paramNames [PanSpeed,Erode,Emissive,MasterAlpha]}
  NamedRerouteDeclaration(None=@n33.PanSpeed) {name=PanSpeed, nodeColor=<rgba 1 0.6563 0 1>}
  NamedRerouteDeclaration(None=@n33.Erode) {name=Erode, nodeColor=<rgba 1 0 0.3047 1>}
  n34 = Power(Base=NamedRerouteUsage(), Exp=ScalarParameter() {defaultValue=1, group=WPO, parameterName=VertexMask Pow})
  n35 = DynamicParameter() {parameterIndex=1, defaultValue=<rgba 1 0.0125 1 0.0125>, paramNames [VelX,VelY,VelZ,VelInfluence]}
  n36 = Clamp(None=@n35.VelInfluence) {maxDefault=4000}
  NamedRerouteDeclaration(None=Multiply(A=Multiply(A=@n34, B=@n36) {constA=0.0125}, B=Multiply(B=@n35.RGB) {constA=-1})) {name=CharacterMovement, nodeColor=<rgba 0 1 0.0156 1>}
  NamedRerouteDeclaration(None=@n33.MasterAlpha) {name=MasterAlpha, nodeColor=<rgba 0.7344 0 1 1>}
  n37 = AppendVector(A=TextureCoordinate(), B=ComponentMask(None=TextureCoordinate() {coordinateIndex=1}) {r=true})
  n38 = Constant3Vector() {constant=<rgba 0 -1 0 0>}
  NamedRerouteDeclaration(None=TransformPosition(None=Add(A=@n37, B=@n38)) {transformType=TRANSFORMPOSSOURCE_World}) {name=PivotPosition, nodeColor=<rgba 0 1 0.2031 1>}
  Power(Base=OneMinus(None=VertexColor().G)) {constExponent=8}
  n39 = Multiply(A=@n25)
  n40 = ScalarParameter() {defaultValue=15, group=Alpha, parameterName=DepthFadeDistance}
  ShadingPathSwitch(
   Default=StaticSwitchParameter(True=DepthFade(Opacity=@n39, FadeDistance=@n40), False=@n39) {group=Alpha, parameterName=UseDepthFade}
   Mobile=@n39
  ) {desc=Migrated from FeatureLevelSwitch}
 ]
}
